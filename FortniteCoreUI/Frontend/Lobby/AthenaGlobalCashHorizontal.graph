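# Horizontal lobby widget, derived from FortGlobalCashWidget, that displays the global cash (gold bars) balance.
# An HBox lays out a tinted name label, gold bar icon and count text, using widgets inherited from the parent class.
# The graph only toggles visibility: it collapses the widget when a query starts and shows it on success.
# It ignores the returned balance and has no failure handler, leaving text updates outside this Blueprint.
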WidgetBlueprint /FortniteCoreUI/Frontend/Lobby/AthenaGlobalCashHorizontal
graphs [EventGraph {nodes [
   Event OnQueryStarted() {is=title K2Node_Event}
   > SetVisibility(InVisibility=Collapsed)
   Event OnQueryCompletedSuccessfully!() -> (OutputDelegate, then, GlobalCashBalance:int=0) {is=title K2Node_Event}
   > SetVisibility(InVisibility=SelfHitTestInvisible)
  ]}]
variables []
widgets {
 info {inheritedWidgetCount=4, parentClass=/Script/FortniteUI.FortGlobalCashWidget, rootWidgetClass=&HorizontalBox, widgetCount=7}
 widgets [
  HBox_Content {widgetClassPath=&HorizontalBox, bIsVariable=true, widget=HBox_Content}
  Border_0 {
   widgetClassPath=&Border
   parent=HBox_Content
   widget=Border_0
   properties {
    background {
     bIsDynamicallyLoaded=false
     drawAs=NoDrawType
     imageType=NoImage
     mirroring=NoMirror
     resourceName=None
     resourceObject=null
     tiling=NoTile
     imageSize=<32 32>
     margin=<ltrb 0 0 0 0>
     outlineSettings {
      bUseBrushTransparency=false
      roundingType=HalfHeightRadius
      width=0
      color {colorUseRule=UseColor_Specified, specifiedColor=<rgba 0 0 0 0>}
      cornerRadii=<0 0 0 0>
     }
     tintColor {colorUseRule=UseColor_Specified, specifiedColor=<rgba 1 1 1 1>}
     uVRegion {bIsValid=false, max=<0 0>, min=<0 0>}
    }
    contentColorAndOpacity=<rgba 0.0331 0.7011 1 1>
    padding=<ltrb 0 0 0 0>
   }
   slot {class=HorizontalBoxSlot, properties {horizontalAlignment=HAlign_Left, verticalAlignment=VAlign_Center, padding=<ltrb 0 0 6 0>}}
  }
  Text_ResourceName {
   widgetClassPath=&CommonTextBlock
   bInherited=true
   parent=Border_0
   widget=Text_ResourceName
   properties {
    lineHeightPercentage=1.28
    style=/FortUILibrary/Styles/Typography/_TextStyle_Body2._TextStyle_Body2_C
    text="{GOLD}"
    textTransformPolicy=ToUpper
    font {
     bForceMonospaced=false
     bMaterialIsStencil=false
     fontMaterial=null
     fontObject=&BurbankSmall
     letterSpacing=0
     monospacedWidth=1
     size=15
     skewAmount=0
     typefaceFontName=Bold
     outlineSettings {
      bApplyOutlineToDropShadows=false
      bMiteredCorners=false
      bSeparateFillAlpha=false
      outlineMaterial=null
      outlineSize=0
      outlineColor=<rgba 0 0 0 1>
     }
    }
    shadowOffset=<0 0>
    strikeBrush {
     bIsDynamicallyLoaded=false
     drawAs=NoDrawType
     imageType=NoImage
     mirroring=NoMirror
     resourceName=None
     resourceObject=null
     tiling=NoTile
     imageSize=<32 32>
     margin=<ltrb 0 0 0 0>
     outlineSettings {
      bUseBrushTransparency=false
      roundingType=HalfHeightRadius
      width=0
      color {colorUseRule=UseColor_Specified, specifiedColor=<rgba 0 0 0 0>}
      cornerRadii=<0 0 0 0>
     }
     tintColor {colorUseRule=UseColor_Specified, specifiedColor=<rgba 1 1 1 1>}
     uVRegion {bIsValid=false, max=<0 0>, min=<0 0>}
    }
   }
   slot {class=BorderSlot, properties {padding=<ltrb 0 0 0 0>}}
  }
  HorizontalBox_0 {widgetClassPath=&HorizontalBox, parent=HBox_Content, widget=HorizontalBox_0, slot {class=HorizontalBoxSlot}}
  LazyImage_ResourceIcon {
   widgetClassPath=&FortLazyImage
   bInherited=true
   parent=HorizontalBox_0
   widget=LazyImage_ResourceIcon
   properties {brush {
     bIsDynamicallyLoaded=false
     drawAs=Image
     imageType=NoImage
     mirroring=NoMirror
     resourceName=None
     resourceObject=/Game/UI/Foundation/Textures/Icons/Athena/T-T-Icon-BR-GoldBars-UI-Icon
     tiling=NoTile
     imageSize=<32 32>
     margin=<ltrb 0 0 0 0>
     outlineSettings {
      bUseBrushTransparency=false
      roundingType=HalfHeightRadius
      width=0
      color {colorUseRule=UseColor_Specified, specifiedColor=<rgba 0 0 0 0>}
      cornerRadii=<0 0 0 0>
     }
     tintColor {colorUseRule=UseColor_Specified, specifiedColor=<rgba 1 1 1 1>}
     uVRegion {bIsValid=false, max=<0 0>, min=<0 0>}
    }}
   slot {class=HorizontalBoxSlot, properties {horizontalAlignment=HAlign_Left, verticalAlignment=VAlign_Center, padding=<ltrb 0 0 6 0>}}
  }
  Text_ResourceCount {
   widgetClassPath=&CommonTextBlock
   bInherited=true
   parent=HorizontalBox_0
   widget=Text_ResourceCount
   properties {
    lineHeightPercentage=1.29
    style=/FortUILibrary/Styles/Typography/_TextStyle_Body1._TextStyle_Body1_C
    text="{0}"
    font {
     bForceMonospaced=false
     bMaterialIsStencil=false
     fontMaterial=null
     fontObject=&BurbankSmall
     letterSpacing=0
     monospacedWidth=1
     size=18
     skewAmount=0
     typefaceFontName=Bold
     outlineSettings {
      bApplyOutlineToDropShadows=false
      bMiteredCorners=false
      bSeparateFillAlpha=false
      outlineMaterial=null
      outlineSize=0
      outlineColor=<rgba 0 0 0 1>
     }
    }
    shadowOffset=<0 0>
    strikeBrush {
     bIsDynamicallyLoaded=false
     drawAs=NoDrawType
     imageType=NoImage
     mirroring=NoMirror
     resourceName=None
     resourceObject=null
     tiling=NoTile
     imageSize=<32 32>
     margin=<ltrb 0 0 0 0>
     outlineSettings {
      bUseBrushTransparency=false
      roundingType=HalfHeightRadius
      width=0
      color {colorUseRule=UseColor_Specified, specifiedColor=<rgba 0 0 0 0>}
      cornerRadii=<0 0 0 0>
     }
     tintColor {colorUseRule=UseColor_Specified, specifiedColor=<rgba 1 1 1 1>}
     uVRegion {bIsValid=false, max=<0 0>, min=<0 0>}
    }
   }
   slot {class=HorizontalBoxSlot, properties {horizontalAlignment=HAlign_Center, verticalAlignment=VAlign_Center}}
  }
  TutorialHighlight {widgetClassPath=&TutorialHighlight}
 ]
}
